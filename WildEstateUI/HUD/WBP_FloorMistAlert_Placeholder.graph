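# Placeholder HUD alert for Wild Estate floor mist, warning "Hounds Incoming" with a Friendly or Danger headline.
# On Construct, IsFriendly checks the controlled pawn's gameplay tag to choose the headline and green or red icon outline, then plays a scale-in Intro animating text and icon material parameters.
# Five seconds after initialization, it clears the current gameplay alert state through the global event router.
# IsHostile is unused, and IconMaterialRef is never assigned here.
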WidgetBlueprint /WildEstateUI/HUD/WBP_FloorMistAlert_Placeholder
animations [Intro {
  binding_count=4
  length_seconds=0.65
  object=&self:Intro
  sequence {bindings [
    Text_MainText_Danger {
     id=#1
     possessed_class=CommonTextBlock
     tracks [{
       type=MovieScene2DTransformTrack
       sections [{
         type=MovieScene2DTransformSection
         end_frame=13
         start_frame=0
         channels [
          Scale.X {type=MovieSceneScriptingFloatChannel, keys [[6,0],[9,1.4],[11,1]]}
          Scale.Y {type=MovieSceneScriptingFloatChannel, keys [[6,0],[9,0.5],[11,1]]}
         ]
        }]
      }]
    }
    Image_HoundIcon {
     id=#2
     possessed_class=Image
     tracks [{
       type=MovieSceneWidgetMaterialTrack
       sections [{
         type=MovieSceneParameterSection
         end_frame=13
         start_frame=0
         channels [
          Circle scale {type=MovieSceneScriptingFloatChannel, keys [[0,0],[4,2.3],[7,1.6]]}
          Icon Scale {type=MovieSceneScriptingFloatChannel, keys [[4,0],[7,0.7]]}
          Icon X pos {type=MovieSceneScriptingFloatChannel, keys [[4,0.5],[7,-0.1551],[9,-0.0951]]}
          Icon Y pos {type=MovieSceneScriptingFloatChannel, keys [[4,-0.3],[7,0],[9,-0.0249]]}
         ]
        }]
      }]
    }
    Text_MainText_1 {
     id=#3
     possessed_class=CommonTextBlock
     tracks [{
       type=MovieScene2DTransformTrack
       sections [{
         type=MovieScene2DTransformSection
         end_frame=13
         start_frame=0
         channels [
          Scale.X {type=MovieSceneScriptingFloatChannel, keys [[9,0],[12,1]]}
          Scale.Y {type=MovieSceneScriptingFloatChannel, keys [[9,0],[12,1]]}
         ]
        }]
      }]
    }
    Text_MainText_Friendly {
     id=#4
     possessed_class=CommonTextBlock
     tracks [{
       type=MovieScene2DTransformTrack
       sections [{
         type=MovieScene2DTransformSection
         end_frame=13
         start_frame=0
         channels [
          Scale.X {type=MovieSceneScriptingFloatChannel, keys [[6,0],[9,1.4],[11,1]]}
          Scale.Y {type=MovieSceneScriptingFloatChannel, keys [[6,0],[9,0.5],[11,1]]}
         ]
        }]
      }]
    }
   ]}
 }]
graphs [
 EventGraph {nodes [
   Event Construct()
   > n1 = Branch(Condition=IsFriendly!() -> (Friendly:bool))
   > Set Visibility(Visibility=SelfHitTestInvisible, self=Get Text_MainText_Friendly() -> (Text_MainText_Friendly:Common Text Object Reference))
   > Set Visibility(Visibility=Collapsed, self=Get Text_MainText_Danger() -> (Text_MainText_Danger:Common Text Object Reference))
   > n2 = SetVectorParameterValue(
    self=Get IconMaterialRef() -> (IconMaterialRef:Material Instance Dynamic Object Reference)
    ParameterName=OutlineColor
    Value=(R=0.2623,G=0.7454,B=0.013,A=1)
   )
   PlayAnimation(execute=[@n2.then, @n3.then], InAnimation=Get Intro())
   Set Visibility(execute=@n1.else, Visibility=Collapsed, self=Get Text_MainText_Friendly())
   > Set Visibility(Visibility=SelfHitTestInvisible, self=Get Text_MainText_Danger())
   > n3 = SetVectorParameterValue(self=Get IconMaterialRef(), ParameterName=OutlineColor, Value=(R=0.7379,G=0.0482,B=0.0075,A=1))
   Event OnInitialized()
   > Delay(Duration=5)
   > ClearEventState!(
    execute
    self:Gameplay Event Router Component Object Reference=GetGlobalEventRouter()
    EventType:Script Struct Object Reference=/Script/BattleRoyaleGameplayAlerts.FortEvent_CurrentGameplayAlertChanged
   ) -> (then)
  ]}
 IsFriendly {nodes [
   IsFriendly() {is=title K2Node_FunctionEntry}
   > n1 = Cast To FortPlayerPawnAthena(Object=Get Controlled Pawn!(self:Controller Object Reference=GetOwningPlayer()) -> (ReturnValue:Pawn Object Reference) {is=title K2Node_CallFunction})
   > n2 = HasMatchingGameplayTag!(
    execute:Exec
    self:Object Reference=@n1
    TagToCheck:Gameplay Tag Structure=Get GameplayTag_Friendly() -> (GameplayTag_Friendly:Gameplay Tag Structure)
   ) -> (then:Exec, ReturnValue:bool) {is=title K2Node_Message}
   > Return Node!(execute, Friendly:bool=@n2)
  ]}
 IsHostile {nodes [
   IsHostile() {is=title K2Node_FunctionEntry}
   > n1 = Cast To FortPlayerPawnAthena(Object=Get Controlled Pawn(self=GetOwningPlayer()))
   > n2 = HasMatchingGameplayTag(self=@n1, TagToCheck=Get GameplayTag_Hostile() -> (GameplayTag_Hostile:Gameplay Tag Structure))
   > Return Node!(execute, Hostile:bool=@n2)
  ]}
]
variables [
 GameplayTag_Friendly {type=GameplayTag, default {tagName=Gameplay.Status.WildEstate.FloorMist.InArea.Friendly}}
 GameplayTag_Hostile {type=GameplayTag, default {tagName=Gameplay.Status.WildEstate.FloorMist.InArea.Hostile}}
 IconMaterialRef {type=MaterialInstanceDynamic}
]
widgets {
 info {inheritedWidgetCount=0, parentClass=&UserWidget, rootWidgetClass=&GridPanel, widgetCount=8}
 widgets [
  GridPanel_0 {widgetClassPath=&GridPanel, widget=GridPanel_0}
  Overlay_0 {
   widgetClassPath=&Overlay
   parent=GridPanel_0
   widget=Overlay_0
   slot {class=GridSlot, properties {horizontalAlignment=HAlign_Center, verticalAlignment=VAlign_Center, nudge=<0 -32>}}
  }
  Image_HoundIcon {
   widgetClassPath=&Image
   bIsVariable=true
   parent=Overlay_0
   widget=Image_HoundIcon
   properties {
    visibility=SelfHitTestInvisible
    brush {
     bIsDynamicallyLoaded=false
     drawAs=Image
     imageType=FullColor
     mirroring=NoMirror
     resourceName=None
     resourceObject=/WildEstateUI/HUD/Materials/MI_UI_Frame
     tiling=NoTile
     imageSize=<180 180>
     margin=<ltrb 0 0 0 0>
     outlineSettings {
      bUseBrushTransparency=false
      roundingType=HalfHeightRadius
      width=0
      color {colorUseRule=UseColor_Specified, specifiedColor=<rgba 0 0 0 0>}
      cornerRadii=<0 0 0 0>
     }
     tintColor {colorUseRule=UseColor_Specified, specifiedColor=<rgba 1 1 1 1>}
     uVRegion {bIsValid=false, max=<0 0>, min=<0 0>}
    }
   }
   slot {class=OverlaySlot, properties {horizontalAlignment=HAlign_Fill, verticalAlignment=VAlign_Fill}}
  }
  Overlay_MainText {widgetClassPath=&Overlay, parent=GridPanel_0, widget=Overlay_MainText, slot {class=GridSlot, properties {row=1, nudge=<0 -56>}}}
  Text_MainText_Danger {
   widgetClassPath=&CommonTextBlock
   bIsVariable=true
   parent=Overlay_MainText
   widget=Text_MainText_Danger
   properties {
    justification=Center
    text="DANGER!"
    textTransformPolicy=ToUpper
    colorAndOpacity {colorUseRule=UseColor_Specified, specifiedColor=<rgba 0.7379 0.0482 0.0075 1>}
    font {
     bForceMonospaced=false
     bMaterialIsStencil=false
     fontMaterial=null
     fontObject=&Burbank
     letterSpacing=0
     monospacedWidth=1
     size=48
     skewAmount=0
     typefaceFontName=BurbankRegular-Black
     outlineSettings {
      bApplyOutlineToDropShadows=false
      bMiteredCorners=false
      bSeparateFillAlpha=false
      outlineMaterial=null
      outlineSize=2
      outlineColor=<rgba 0 0 0 1>
     }
    }
    margin=<ltrb 0 0 0 -16>
    shadowColorAndOpacity=<rgba 0 0 0 1>
    shadowOffset=<0 2>
   }
   slot {class=OverlaySlot, properties {horizontalAlignment=HAlign_Fill, verticalAlignment=VAlign_Fill}}
  }
  Text_MainText_Friendly {
   widgetClassPath=&CommonTextBlock
   bIsVariable=true
   parent=Overlay_MainText
   widget=Text_MainText_Friendly
   properties {
    justification=Center
    text="Friendly!"
    textTransformPolicy=ToUpper
    colorAndOpacity {colorUseRule=UseColor_Specified, specifiedColor=<rgba 0.2219 0.8632 0.0542 1>}
    font {
     bForceMonospaced=false
     bMaterialIsStencil=false
     fontMaterial=null
     fontObject=&Burbank
     letterSpacing=0
     monospacedWidth=1
     size=48
     skewAmount=0
     typefaceFontName=BurbankRegular-Black
     outlineSettings {
      bApplyOutlineToDropShadows=false
      bMiteredCorners=false
      bSeparateFillAlpha=false
      outlineMaterial=null
      outlineSize=2
      outlineColor=<rgba 0 0 0 1>
     }
    }
    margin=<ltrb 0 0 0 -16>
    shadowColorAndOpacity=<rgba 0 0 0 1>
    shadowOffset=<0 2>
   }
   slot {class=OverlaySlot, properties {horizontalAlignment=HAlign_Fill, verticalAlignment=VAlign_Fill}}
  }
  Overlay_SecondaryText {
   widgetClassPath=&Overlay
   parent=GridPanel_0
   widget=Overlay_SecondaryText
   slot {class=GridSlot, properties {row=2, nudge=<0 -56>}}
  }
  Text_MainText_1 {
   widgetClassPath=&CommonTextBlock
   bIsVariable=true
   parent=Overlay_SecondaryText
   widget=Text_MainText_1
   properties {
    justification=Center
    text=Hounds Incoming
    colorAndOpacity {colorUseRule=UseColor_Specified, specifiedColor=<rgba 0.8632 0.8632 0.8632 1>}
    font {
     bForceMonospaced=false
     bMaterialIsStencil=false
     fontMaterial=null
     fontObject=&Burbank
     letterSpacing=80
     monospacedWidth=1
     size=24
     skewAmount=0
     typefaceFontName=BurbankRegular-Black
     outlineSettings {
      bApplyOutlineToDropShadows=false
      bMiteredCorners=false
      bSeparateFillAlpha=false
      outlineMaterial=null
      outlineSize=2
      outlineColor=<rgba 0 0 0 1>
     }
    }
    shadowColorAndOpacity=<rgba 0 0 0 1>
    shadowOffset=<0 2>
   }
   slot {class=OverlaySlot, properties {horizontalAlignment=HAlign_Fill, verticalAlignment=VAlign_Fill}}
  }
 ]
}
